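# Sound cue for a looping ambient spot of water slapping against a pontoon in Asteria's IndustrialTown POI.
# A looping wave player feeds an attenuation node using a shared Apollo very-close ocean-shore preset, then a root Modulator left at default settings.
# The cue lowers pitch and routes to the ambient water sound class.
# It also stores attenuation and single-instance concurrency override values, though neither override is shown enabled.
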SoundCue /Game/Sounds/Asteria/POI/IndustrialTown/Spots/Asteria_Water_Pontoon_Cue
root=Modulator_0
nodes [
 {
  id=Attenuation_6
  children [WavePlayer_7]
  properties {attenuationSettings=/Game/Athena/Sounds/Water/Attenuations/Apollo_Water_Ocean_Shore_Very_Close_Att}
 }
 {id=Modulator_0, children [Attenuation_6]}
 {
  id=WavePlayer_7
  properties {bLooping=true, soundWaveAssetPtr=&/Spot_Water_Slapping_Against_Pontoon_Loop}
 }
]
properties {
 maxDistance=2500
 pitchMultiplier=0.7
 soundClassObject=/Game/Sounds/FortSoundClassesAndModes/Fort_Water_Ambient
 volumeMultiplier=1
 attenuationOverrides {
  bAttenuateWithLPF=true
  falloffDistance=6000
  lPFFrequencyAtMax=500
  nonSpatializedRadiusStart=700
  stereoSpread=1000
  attenuationShapeExtents {x=925}
 }
 concurrencyOverrides {maxCount=1, resolutionRule=PreventNew}
}
