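# Looping ambient spot sound cue for manhole steam in the IndustrialTown POI.
# A mixer layers one of two randomly chosen AC duct air loops with a pipe fitting loop, the latter heavily volume-boosted and slightly pitch-varied by its own modulator.
# A root modulator varies the whole mix.
# The cue uses dedicated attenuation and the Athena ambient sound class, with a low overall volume and a small concurrency cap.
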SoundCue /Game/Sounds/Asteria/POI/IndustrialTown/Spots/SC_Spot_Manhole_Steam_Loop
root=Modulator_2
nodes [
 {id=Mixer_0, children [Random_1,Modulator_3], properties {inputVolume [1,1]}}
 {id=Modulator_2, children [Mixer_0], properties {pitchMax=1.1, pitchMin=0.9, volumeMax=1}}
 {id=Modulator_3, children [WavePlayer_5], properties {pitchMax=1, pitchMin=0.9, volumeMax=11, volumeMin=11}}
 {id=Random_1, children [WavePlayer_2,WavePlayer_3], properties {weights [1,1]}}
 {id=WavePlayer_2, properties {bLooping=true, soundWaveAssetPtr=&/Spot_AC_Duct_Air_01}}
 {id=WavePlayer_3, properties {bLooping=true, soundWaveAssetPtr=&/Spot_AC_Duct_Air_03}}
 {id=WavePlayer_5, properties {bLooping=true, soundWaveAssetPtr=&/Spot_Outpost_Pipe_Fitting_Loop}}
]
properties {
 attenuationSettings=&/ATT_Spot_Manhole_Steam_Loop
 bOverrideConcurrency=true
 maxDistance=2000
 soundClassObject=&Fort_Ambient_Athena
 volumeMultiplier=0.175
 concurrencyOverrides {maxCount=3}
}
