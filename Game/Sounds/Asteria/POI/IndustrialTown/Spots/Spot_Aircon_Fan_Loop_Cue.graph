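# Looping spot ambience cue for an air-conditioner fan in the Asteria Industrial Town POI, routed to the Fort_Ambient_Athena sound class.
# A Random node picks one of three equally weighted looping fan waves, which a Modulator varies before an Attenuation node and Mixer with one empty input.
# Attenuation adds distance low-pass, occlusion and listener focus with reverb send off, and a concurrency override caps simultaneous instances at three.
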SoundCue /Game/Sounds/Asteria/POI/IndustrialTown/Spots/Spot_Aircon_Fan_Loop_Cue
root=Mixer_1
nodes [
 {
  id=Attenuation_1
  children [Modulator_0]
  properties {
   bOverrideAttenuation=true
   attenuationOverrides {
    bAttenuateWithLPF=true
    bEnableFocusInterpolation=true
    bEnableListenerFocus=true
    bEnableOcclusion=true
    bEnableReverbSend=false
    falloffDistance=1400
    focusReleaseInterpSpeed=0.5
    lPFFrequencyAtMax=2000
    lPFRadiusMax=3000
    lPFRadiusMin=500
    nonFocusVolumeAttenuation=0.5
    occlusionInterpolationTime=1
    occlusionLowPassFilterFrequency=2000
    occlusionTraceChannel=ECC_GameTraceChannel11
    occlusionVolumeAttenuation=0.5
    stereoSpread=512
    attenuationShapeExtents {x=160}
   }
  }
 }
 {id=Mixer_1, children [Attenuation_1, null], properties {inputVolume [1,1]}}
 {id=Modulator_0, children [Random_1]}
 {id=Random_1, children [WavePlayer_3,WavePlayer_4,WavePlayer_5], properties {weights [1,1,1]}}
 {id=WavePlayer_3, properties {bLooping=true, soundWaveAssetPtr=&/Spot_AC_Fan_03}}
 {id=WavePlayer_4, properties {bLooping=true, soundWaveAssetPtr=&/Spot_AC_Fan_01}}
 {id=WavePlayer_5, properties {bLooping=true, soundWaveAssetPtr=&/Spot_AC_Fan_02}}
]
properties {
 bOverrideConcurrency=true
 maxDistance=1560
 soundClassObject=&Fort_Ambient_Athena
 volumeMultiplier=0.376
 attenuationOverrides {attenuationShapeExtents {x=0}}
 concurrencyOverrides {maxCount=3}
}
