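# Looping electrical hum spot ambience for the IndustrialTown POI in Asteria, routed to the Athena ambient sound class.
# A looping wave player feeds a default-range Modulator, then an Attenuation node with overridden settings.
# That attenuation adds distance low-pass filtering, occlusion on a custom trace channel, focus interpolation and stereo spread, with reverb send off.
# A concurrency override caps simultaneous instances to a small count.
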SoundCue /Game/Sounds/Asteria/POI/IndustrialTown/Spots/Spot_Electrical_Hum_Cue
root=Attenuation_1
nodes [
 {
  id=Attenuation_1
  children [Modulator_0]
  properties {
   bOverrideAttenuation=true
   attenuationOverrides {
    bEnableFocusInterpolation=true
    bEnableReverbSend=false
    falloffDistance=1700
    focusReleaseInterpSpeed=0.5
    lPFFrequencyAtMax=2000
    lPFRadiusMax=3000
    lPFRadiusMin=500
    nonFocusVolumeAttenuation=0.5
    occlusionInterpolationTime=1
    occlusionLowPassFilterFrequency=2000
    occlusionTraceChannel=ECC_GameTraceChannel11
    occlusionVolumeAttenuation=0.5
    stereoSpread=512
    attenuationShapeExtents=<1250 250>
   }
  }
 }
 {id=Modulator_0, children [WavePlayer_1]}
 {id=WavePlayer_1, properties {bLooping=true, soundWaveAssetPtr=&/Spot_Electrical_Hum_Int_01}}
]
properties {
 bOverrideConcurrency=true
 maxDistance=2950
 soundClassObject=&Fort_Ambient_Athena
 volumeMultiplier=0.501
 attenuationOverrides {attenuationShapeExtents {x=0}}
 concurrencyOverrides {maxCount=3}
}
