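# Sound cue for a distant industrial drill spot ambience in the Asteria Industrial Town point of interest.
# A Random node picks among six equally weighted distant industrial waves, and a Modulator above it varies pitch on each play.
# It plays quietly in the outdoor ambient sound class, using a wide one-shot ambient attenuation with air absorption and overrides for falloff, extents and stereo spread.
# Bus and submix sends are disabled.
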SoundCue /Game/Sounds/Asteria/POI/IndustrialTown/Spots/Spot_Industrial_Drill_Distant_Cue
root=Modulator_0
nodes [
 {id=Modulator_0, children [Random_0], properties {pitchMin=0.9}}
 {id=Random_0, children [WavePlayer_0,WavePlayer_1,WavePlayer_2,WavePlayer_3,WavePlayer_4,WavePlayer_5], properties {weights [1,1,1,1,1,1]}}
 {id=WavePlayer_0, properties {soundWaveAssetPtr=&/Spot_Amb_Industrial_Distant_10}}
 {id=WavePlayer_1, properties {soundWaveAssetPtr=&/Spot_Amb_Industrial_Distant_11}}
 {id=WavePlayer_2, properties {soundWaveAssetPtr=&/Spot_Amb_Industrial_Distant_12}}
 {id=WavePlayer_3, properties {soundWaveAssetPtr=&/Spot_Amb_Industrial_Distant_13}}
 {id=WavePlayer_4, properties {soundWaveAssetPtr=&/Spot_Amb_Industrial_Distant_18}}
 {id=WavePlayer_5, properties {soundWaveAssetPtr=&/Spot_Amb_Industrial_Distant_17}}
]
properties {
 attenuationSettings=/Game/Sounds/Attenuation/Fort_TileAmbient_OneShot_Wider_AirAbsorbtion_Att
 bEnableBusSends=false
 bEnableSubmixSends=false
 maxDistance=8000
 soundClassObject=&Fort_Ambient_Outdoors_Athena
 volumeMultiplier=0.251
 attenuationOverrides {falloffDistance=6000, stereoSpread=10000, attenuationShapeExtents {x=2000}}
}
