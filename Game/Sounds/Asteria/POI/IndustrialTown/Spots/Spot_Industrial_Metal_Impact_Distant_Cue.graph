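# Sound cue for distant industrial metal-impact spot ambience at the Industrial Town POI.
# A Random node picks among ten equally weighted distant industrial wave variations, and a Modulator adds slight pitch variation.
# It plays quietly in the outdoor ambient sound class with a less-positional one-shot ambient attenuation.
# Bus and submix sends are disabled.
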SoundCue /Game/Sounds/Asteria/POI/IndustrialTown/Spots/Spot_Industrial_Metal_Impact_Distant_Cue
root=Modulator_0
nodes [
 {id=Modulator_0, children [Random_0], properties {pitchMin=0.9}}
 {
  id=Random_0
  children [
   WavePlayer_11
   WavePlayer_10
   WavePlayer_9
   WavePlayer_8
   WavePlayer_7
   WavePlayer_6
   WavePlayer_12
   WavePlayer_13
   WavePlayer_14
   WavePlayer_15
  ]
  properties {weights [1,1,1,1,1,1,1,1,1,1]}
 }
 {id=WavePlayer_10, properties {soundWaveAssetPtr=&/Spot_Amb_Industrial_Distant_15}}
 {id=WavePlayer_11, properties {soundWaveAssetPtr=&/Spot_Amb_Industrial_Distant_09}}
 {id=WavePlayer_12, properties {soundWaveAssetPtr=&/Spot_Amb_Industrial_Distant_08}}
 {id=WavePlayer_13, properties {soundWaveAssetPtr=&/Spot_Amb_Industrial_Distant_05}}
 {id=WavePlayer_14, properties {soundWaveAssetPtr=&/Spot_Amb_Industrial_Distant_04}}
 {id=WavePlayer_15, properties {soundWaveAssetPtr=&/Spot_Amb_Industrial_Distant_03}}
 {id=WavePlayer_6, properties {soundWaveAssetPtr=&/Spot_Amb_Industrial_Distant_22}}
 {id=WavePlayer_7, properties {soundWaveAssetPtr=&/Spot_Amb_Industrial_Distant_21}}
 {id=WavePlayer_8, properties {soundWaveAssetPtr=&/Spot_Amb_Industrial_Distant_20}}
 {id=WavePlayer_9, properties {soundWaveAssetPtr=&/Spot_Amb_Industrial_Distant_19}}
]
properties {
 attenuationSettings=/Game/Sounds/Attenuation/Fort_TileAmbient_OneShot_LessPositional_Log_Att
 bEnableBusSends=false
 bEnableSubmixSends=false
 maxDistance=6200
 soundClassObject=&Fort_Ambient_Outdoors_Athena
 volumeMultiplier=0.301
 attenuationOverrides {falloffDistance=6000, stereoSpread=10000, attenuationShapeExtents {x=2000}}
}
