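# Sound cue for a distant steam whistle horn one-shot, an outdoor ambient spot sound for the Asteria IndustrialTown POI.
# A single wave player feeds a modulator that varies pitch on each play.
# It uses a less-positional ambient one-shot attenuation asset and the Athena outdoor ambient sound class, with bus and submix sends disabled.
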SoundCue /Game/Sounds/Asteria/POI/IndustrialTown/Spots/Spot_Industrial_SteamWhistleHorn_Distant_Cue
root=Modulator_0
nodes [
 {id=Modulator_0, children [WavePlayer_11], properties {pitchMin=0.9}}
 {id=WavePlayer_11, properties {soundWaveAssetPtr=&/Spot_Amb_Industrial_Distant_07}}
]
properties {
 attenuationSettings=/Game/Sounds/Attenuation/Fort_TileAmbient_OneShot_LessPositional_Log_Att
 bEnableBusSends=false
 bEnableSubmixSends=false
 maxDistance=6200
 soundClassObject=&Fort_Ambient_Outdoors_Athena
 volumeMultiplier=1.253
 attenuationOverrides {falloffDistance=6000, stereoSpread=10000, attenuationShapeExtents {x=2000}}
}
