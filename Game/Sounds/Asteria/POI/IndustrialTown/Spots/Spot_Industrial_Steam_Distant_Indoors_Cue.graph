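# Sound cue for distant industrial steam one-shot spot ambience in the Asteria IndustrialTown POI, classed as indoor ambient one-shots.
# A Random node picks one of five equally weighted distant industrial waves and feeds a Modulator that varies pitch per play.
# It plays at reduced volume through a shared less-positional logarithmic tile-ambient attenuation asset, with bus and submix sends disabled.
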SoundCue /Game/Sounds/Asteria/POI/IndustrialTown/Spots/Spot_Industrial_Steam_Distant_Indoors_Cue
root=Modulator_0
nodes [
 {id=Modulator_0, children [Random_0], properties {pitchMin=0.9}}
 {id=Random_0, children [WavePlayer_11, WavePlayer_10, WavePlayer_9, WavePlayer_8, WavePlayer_0], properties {weights [1,1,1,1,1]}}
 {id=WavePlayer_0, properties {soundWaveAssetPtr=&/Spot_Amb_Industrial_Distant_16}}
 {id=WavePlayer_10, properties {soundWaveAssetPtr=&/Spot_Amb_Industrial_Distant_02}}
 {id=WavePlayer_11, properties {soundWaveAssetPtr=&/Spot_Amb_Industrial_Distant_01}}
 {id=WavePlayer_8, properties {soundWaveAssetPtr=&/Spot_Amb_Industrial_Distant_14}}
 {id=WavePlayer_9, properties {soundWaveAssetPtr=&/Spot_Amb_Industrial_Distant_06}}
]
properties {
 attenuationSettings=/Game/Sounds/Attenuation/Fort_TileAmbient_OneShot_LessPositional_Log_Att
 bEnableBusSends=false
 bEnableSubmixSends=false
 maxDistance=6200
 soundClassObject=/Game/Sounds/FortSoundClassesAndModes/Fort_Ambient_OneShots_Indoors
 volumeMultiplier=0.251
 attenuationOverrides {falloffDistance=6000, stereoSpread=10000, attenuationShapeExtents {x=2000}}
}
