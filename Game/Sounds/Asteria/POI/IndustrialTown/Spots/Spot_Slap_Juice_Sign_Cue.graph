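# Sound cue for a quiet, looping ambient spot sound of the Slap Juice factory sign in the Industrial Town POI.
# A single looping wave feeds a default-setting Modulator, then an Attenuation node with its own capsule-shaped override.
# Attenuation adds distance low-pass filtering and interpolated listener focus, and disables reverb send.
# The cue plays under the Athena ambient sound class and overrides concurrency to cap simultaneous instances at three.
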SoundCue /Game/Sounds/Asteria/POI/IndustrialTown/Spots/Spot_Slap_Juice_Sign_Cue
root=Attenuation_1
nodes [
 {
  id=Attenuation_1
  children [Modulator_0]
  properties {
   bOverrideAttenuation=true
   attenuationOverrides {
    attenuationShape=Capsule
    bAttenuateWithLPF=true
    bEnableFocusInterpolation=true
    bEnableListenerFocus=true
    bEnableReverbSend=false
    falloffDistance=1700
    focusReleaseInterpSpeed=0.5
    lPFFrequencyAtMax=2000
    lPFRadiusMax=3000
    lPFRadiusMin=500
    nonFocusVolumeAttenuation=0.5
    occlusionInterpolationTime=1
    occlusionLowPassFilterFrequency=2000
    occlusionTraceChannel=ECC_GameTraceChannel11
    occlusionVolumeAttenuation=0.5
    stereoSpread=512
    attenuationShapeExtents=<1250 250>
   }
  }
 }
 {id=Modulator_0, children [WavePlayer_0]}
 {id=WavePlayer_0, properties {bLooping=true, soundWaveAssetPtr=&/Spot_SlapFactory_Sign01}}
]
properties {
 bOverrideConcurrency=true
 maxDistance=2950
 soundClassObject=&Fort_Ambient_Athena
 volumeMultiplier=0.251
 attenuationOverrides {attenuationShapeExtents {x=0}}
 concurrencyOverrides {maxCount=3}
}
